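# Control Rig for the SlowBurn male medium body cosmetic that procedurally poses its digitigrade legs and robe bones.
# Digitigrade calves and heels copy standard leg rotations with rest offsets, then FABRIK places toes relative to the ball bones.
# Spherical pose readers and thigh angles weight robe aims, parent constraints and rotation offsets.
# Calf bend adjusts back robes, frontend idle biases the centre-front aim, and smoothed melee alpha offsets centre and left panels.
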ControlRigBlueprint /BRCosmetics/Characters/Player/Male/Medium/Bodies/M_MED_SlowBurn/Meshes/M_MED_SlowBurn_CtrlRig
graphs [
 RigVMFunctionLibrary {links [], nodes []}
 RigVMModel {nodes [
   RigUnit_BeginExecution = &RigUnit_BeginExecution()
   n1 = &RigUnit_SphericalPoseReader(
    ActiveRegionScaleFactors=(PositiveWidth=2,NegativeWidth=2,PositiveHeight=1.9,NegativeHeight=1)
    ActiveRegionSize=0.1
    Debug=(bDrawDebug=false,bDraw2D=False,bDrawLocalAxes=False,DebugScale=25,DebugSegments=20,DebugThickness=0.25)
    DriverAxis=(X=-1,Y=0,Z=0)
    DriverItem=(Type=Bone,Name="thigh_l")
    FalloffRegionScaleFactors=(PositiveWidth=1,NegativeWidth=1,PositiveHeight=1.9,NegativeHeight=1)
    FalloffSize=0.3
    FlipHeightScaling=false
    FlipWidthScaling=false
    OptionalParentItem=(Type=Bone,Name="None")
    RotationOffset=(X=0,Y=90,Z=68)
    ExecutePin=@RigUnit_BeginExecution.ExecutePin
   )
   > n2 = &RigUnit_SphericalPoseReader(
    ActiveRegionScaleFactors=(PositiveWidth=2,NegativeWidth=2,PositiveHeight=1.9,NegativeHeight=1)
    ActiveRegionSize=0.1
    Debug=(bDrawDebug=false,bDraw2D=False,bDrawLocalAxes=False,DebugScale=25,DebugSegments=20,DebugThickness=0.25)
    DriverAxis=(X=1,Y=0,Z=0)
    DriverItem=(Type=Bone,Name="thigh_r")
    FalloffRegionScaleFactors=(PositiveWidth=1,NegativeWidth=1,PositiveHeight=1.9,NegativeHeight=1)
    FalloffSize=0.3
    FlipHeightScaling=false
    FlipWidthScaling=false
    OptionalParentItem=(Type=Bone,Name="None")
    RotationOffset=(X=180,Y=90,Z=68)
   )
   > n3 = &RigUnit_SphericalPoseReader(
    ActiveRegionScaleFactors=(PositiveWidth=4,NegativeWidth=3.5,PositiveHeight=2,NegativeHeight=3)
    ActiveRegionSize=0.1
    Debug=(bDrawDebug=false,bDraw2D=False,bDrawLocalAxes=False,DebugScale=25,DebugSegments=20,DebugThickness=0.25)
    DriverAxis=(X=-1,Y=0,Z=0)
    DriverItem=(Type=Bone,Name="thigh_l")
    FalloffRegionScaleFactors=(PositiveWidth=3,NegativeWidth=4,PositiveHeight=1,NegativeHeight=1)
    FalloffSize=0.4
    FlipHeightScaling=False
    FlipWidthScaling=False
    OptionalParentItem=(Type=Bone,Name="None")
    RotationOffset=(X=130,Y=-20,Z=0)
   )
   > n4 = &RigUnit_SphericalPoseReader(
    ActiveRegionScaleFactors=(PositiveWidth=4,NegativeWidth=3.5,PositiveHeight=2,NegativeHeight=3)
    ActiveRegionSize=0.1
    Debug=(bDrawDebug=false,bDraw2D=False,bDrawLocalAxes=False,DebugScale=25,DebugSegments=20,DebugThickness=0.25)
    DriverAxis=(X=1,Y=0,Z=0)
    DriverItem=(Type=Bone,Name="thigh_r")
    FalloffRegionScaleFactors=(PositiveWidth=3,NegativeWidth=4,PositiveHeight=1,NegativeHeight=1)
    FalloffSize=0.4
    FlipHeightScaling=False
    FlipWidthScaling=true
    OptionalParentItem=(Type=Bone,Name="None")
    RotationOffset=(X=-50,Y=-20,Z=0)
   )
   > n5 = &RigUnit_SphericalPoseReader(
    ActiveRegionScaleFactors=(PositiveWidth=3.5,NegativeWidth=2.5,PositiveHeight=1.5,NegativeHeight=1.5)
    ActiveRegionSize=0.1
    Debug=(bDrawDebug=false,bDraw2D=False,bDrawLocalAxes=False,DebugScale=25,DebugSegments=20,DebugThickness=0.25)
    DriverAxis=(X=-1,Y=0,Z=0)
    DriverItem=(Type=Bone,Name="thigh_l")
    FalloffRegionScaleFactors=(PositiveWidth=1,NegativeWidth=1,PositiveHeight=1,NegativeHeight=2)
    FalloffSize=0.3
    FlipHeightScaling=false
    FlipWidthScaling=false
    OptionalParentItem=(Type=Bone,Name="None")
    RotationOffset=(X=180,Y=-20,Z=0)
   )
   > n6 = &RigUnit_SphericalPoseReader(
    ActiveRegionScaleFactors=(PositiveWidth=3.5,NegativeWidth=2.5,PositiveHeight=1.5,NegativeHeight=1.5)
    ActiveRegionSize=0.1
    Debug=(bDrawDebug=false,bDraw2D=False,bDrawLocalAxes=False,DebugScale=25,DebugSegments=20,DebugThickness=0.25)
    DriverAxis=(X=1,Y=0,Z=0)
    DriverItem=(Type=Bone,Name="thigh_r")
    FalloffRegionScaleFactors=(PositiveWidth=1,NegativeWidth=1,PositiveHeight=1,NegativeHeight=2)
    FalloffSize=0.3
    FlipHeightScaling=False
    FlipWidthScaling=true
    OptionalParentItem=(Type=Bone,Name="None")
    RotationOffset=(X=0,Y=-20,Z=0)
   )
   RigVMFunction_MathQuaternionToEuler = &RigVMFunction_MathQuaternionToEuler(
    RotationOrder=ZYX
    Value=&RigUnit_GetRelativeTransformForItem(Child=(Type=Bone,Name="thigh_l"), Parent=(Type=Bone,Name="pelvis"), bChildInitial=False, bParentInitial=False).RelativeTransform.Rotation
   )
   Remap = &RigVMFunction_MathDoubleRemap(SourceMaximum=-90, SourceMinimum=-50, TargetMaximum=1, TargetMinimum=0, Value=@RigVMFunction_MathQuaternionToEuler.Result.Z)
   > RigVMVariableNode(Value=@Remap.Result) {variable=L thighVeryForward}
   > RigVMVariableNode(Value=&RigVMFunction_MathDoubleMul(B=-1, A=@RigVMFunction_MathQuaternionToEuler.Result.Z).Result) {variable=L_thigh_Z_value}
   n7 = &RigVMFunction_MathQuaternionToEuler(
    RotationOrder=ZYX
    Value=&RigUnit_GetRelativeTransformForItem(Child=(Type=Bone,Name="thigh_r"), Parent=(Type=Bone,Name="pelvis"), bChildInitial=False, bParentInitial=False).RelativeTransform.Rotation
   )
   Absolute = &RigVMFunction_MathDoubleAbs(Value=@n7.Result.Z)
   > RigVMVariableNode(Value=&RigVMFunction_MathDoubleRemap(SourceMaximum=90, SourceMinimum=150, TargetMaximum=1, TargetMinimum=0, Value=@Absolute.Result).Result) {variable=R thighVeryForward}
   > RigVMVariableNode(Value=&RigVMFunction_MathDoubleSub(A=180, B=&RigVMFunction_MathDoubleAbs(Value=@n7.Result.Z).Result).Result) {variable=R_thigh_Z_value}
   > RigVMVariableNode(Value=@n1.OutputParam) {variable=L thighBack alpha}
   > RigVMVariableNode(Value=@n2.OutputParam) {variable=R thighBack alpha}
   > RigVMVariableNode(Value=@n3.OutputParam) {variable=L thighForward alpha}
   > RigVMVariableNode(Value=@n4.OutputParam) {variable=R thighForward alpha}
   > RigVMVariableNode(Value=@n5.OutputParam) {variable=L thighSide alpha}
   > RigVMVariableNode(Value=@n6.OutputParam) {variable=R thighSide alpha}
   n8 = &RigUnit_GetTransform(Item=(Type=Bone,Name="dyn_digitigrade_calf_l"), bInitial=true)
   n9 = &RigUnit_GetTransform(Item=(Type=Bone,Name="calf_l"), bInitial=true)
   > Set Transform = &RigUnit_SetRotation(
    Item=(Type=Bone,Name="dyn_digitigrade_calf_l")
    Weight=1
    bInitial=False
    Value=&RigVMFunction_MathTransformMakeRelative(
     Global=&RigUnit_GetTransform(Item=(Type=Bone,Name="calf_l"), Space=LocalSpace, bInitial=false).Transform
     Parent=&RigVMFunction_MathTransformMakeRelative(Parent=@n8.Transform, Global=@n9.Transform).Local
    ).Local.Rotation
   )
   n10 = &RigUnit_GetTransform(Item=(Type=Bone,Name="foot_l"), bInitial=true)
   n11 = &RigUnit_GetTransform(Item=(Type=Bone,Name="dyn_digitigrade_heel_l"), bInitial=true)
   > Set Transform_1 = &RigUnit_SetRotation(
    Item=(Type=Bone,Name="dyn_digitigrade_heel_l")
    Weight=1
    bInitial=false
    Value=&RigVMFunction_MathTransformMakeRelative(
     Global=&RigUnit_GetTransform(Item=(Type=Bone,Name="foot_l"), Space=LocalSpace, bInitial=false).Transform
     Parent=&RigVMFunction_MathTransformMakeRelative(Global=@n10.Transform, Parent=@n11.Transform).Local
    ).Local.Rotation
   )
   Multiply = &RigVMFunction_MathTransformMul(
    B=&RigUnit_GetTransform(Item=(Type=Bone,Name="ball_l"), bInitial=false).Transform
    A=&RigVMFunction_MathTransformMakeRelative(
     Parent=&RigUnit_GetTransform(Item=(Type=Bone,Name="ball_l"), bInitial=true).Transform
     Global=&RigUnit_GetTransform(Item=(Type=Bone,Name="dyn_digitigrade_toe_l"), bInitial=true).Transform
    ).Local
   ) {template="Multiply::Execute(in A,in B,out Result)"}
   > /Script/ControlRig.RigUnit_FABRIKItemArray(
    Items=((Type=Bone,Name="dyn_digitigrade_calf_l"),(Type=Bone,Name="dyn_digitigrade_toe_l"))
    MaxIterations=10
    Precision=1
    Weight=1
    bPropagateToChildren=True
    bSetEffectorTransform=True
    EffectorTransform=@Multiply.Result
   ) {is=struct RigVMUnitNode}
   > &RigUnit_SetTransform(Item=(Type=Bone,Name="dyn_digitigrade_toe_l"), Weight=1, bInitial=False, Value=@Multiply.Result)
   n12 = &RigUnit_GetTransform(Item=(Type=Bone,Name="calf_r"), bInitial=true)
   n13 = &RigUnit_GetTransform(Item=(Type=Bone,Name="dyn_digitigrade_calf_r"), bInitial=true)
   > Set Transform_3 = &RigUnit_SetRotation(
    Item=(Type=Bone,Name="dyn_digitigrade_calf_r")
    Weight=1
    bInitial=False
    Value=&RigVMFunction_MathTransformMakeRelative(
     Global=&RigUnit_GetTransform(Item=(Type=Bone,Name="calf_r"), Space=LocalSpace, bInitial=false).Transform
     Parent=&RigVMFunction_MathTransformMakeRelative(Global=@n12.Transform, Parent=@n13.Transform).Local
    ).Local.Rotation
   )
   n14 = &RigUnit_GetTransform(Item=(Type=Bone,Name="foot_r"), bInitial=true)
   n15 = &RigUnit_GetTransform(Item=(Type=Bone,Name="dyn_digitigrade_heel_r"), bInitial=true)
   > Set Transform_1_1 = &RigUnit_SetRotation(
    Item=(Type=Bone,Name="dyn_digitigrade_heel_r")
    Weight=1
    bInitial=false
    Value=&RigVMFunction_MathTransformMakeRelative(
     Global=&RigUnit_GetTransform(Item=(Type=Bone,Name="foot_r"), Space=LocalSpace, bInitial=false).Transform
     Parent=&RigVMFunction_MathTransformMakeRelative(Global=@n14.Transform, Parent=@n15.Transform).Local
    ).Local.Rotation
   )
   n16 = &RigVMFunction_MathTransformMul(
    B=&RigUnit_GetTransform(Item=(Type=Bone,Name="ball_r"), bInitial=false).Transform
    A=&RigVMFunction_MathTransformMakeRelative(
     Global=&RigUnit_GetTransform(Item=(Type=Bone,Name="dyn_digitigrade_toe_r"), bInitial=true).Transform
     Parent=&RigUnit_GetTransform(Item=(Type=Bone,Name="ball_r"), bInitial=true).Transform
    ).Local
   ) {template="Multiply::Execute(in A,in B,out Result)"}
   > &RigUnit_FABRIKItemArray(
    Items=((Type=Bone,Name="dyn_digitigrade_calf_r"),(Type=Bone,Name="dyn_digitigrade_toe_r"))
    MaxIterations=10
    Precision=1
    Weight=1
    bPropagateToChildren=True
    bSetEffectorTransform=True
    EffectorTransform=@n16.Result
   )
   > n17 = &RigUnit_SetTransform(Item=(Type=Bone,Name="dyn_digitigrade_toe_r"), Weight=1, bInitial=False, Value=@n16.Result)
   RigVMFunction_Sequence = &RigVMFunction_Sequence(ExecuteContext=@n17.ExecutePin)
   LessEqual = /Script/RigVM.RigVMFunction_MathDoubleLessEqual(A=RigVMVariableNode() {variable=L_thigh_Z_value}.Value, B=RigVMVariableNode() {variable=R_thigh_Z_value}.Value) {is=struct RigVMUnitNode, template="LessEqual::Execute(in A,in B,out Result)"}
   RigVMFunction_ControlFlowBranch = &RigVMFunction_ControlFlowBranch(Condition=@LessEqual.Result, ExecuteContext=@RigVMFunction_Sequence.A)
   RigVMVariableNode(Value=1, ExecuteContext=@RigVMFunction_ControlFlowBranch.False) {variable=L_thigh_rotated_more_than_R}
   > RigVMVariableNode() {variable=R_thigh_rotated_more_than_L}
   RigVMVariableNode(Value=1, ExecuteContext=@RigVMFunction_ControlFlowBranch.True) {variable=R_thigh_rotated_more_than_L}
   > RigVMVariableNode() {variable=L_thigh_rotated_more_than_R}
   n18 = RigVMVariableNode() {variable=R thighSide alpha}
   n19 = &RigVMFunction_MathDoubleSub(A=1, B=RigVMVariableNode() {variable=L thighSide alpha}.Value)
   n20 = &RigVMFunction_MathDoubleMul(B=RigVMVariableNode() {variable=L_thigh_rotated_more_than_R}.Value, A=RigVMVariableNode() {variable=L thighForward alpha}.Value)
   Add = &RigVMFunction_MathDoubleAdd(
    B=&RigVMFunction_MathDoubleMul(B=&RigVMFunction_MathDoubleClamp(Value=@n19.Result).Result, A=@n20.Result).Result
    A=@n20.Result
   )
   Clamp = &RigVMFunction_MathDoubleClamp(Value=&RigVMFunction_MathDoubleSub(A=1, B=RigVMVariableNode() {variable=R thighSide alpha}.Value).Result)
   n21 = &RigVMFunction_MathDoubleMul(B=RigVMVariableNode() {variable=R_thigh_rotated_more_than_L}.Value, A=RigVMVariableNode() {variable=R thighForward alpha}.Value)
   n22 = &RigVMFunction_MathDoubleMul(B=@Clamp.Result, A=@n21.Result)
   Subtract = &RigVMFunction_MathDoubleSub(
    B=&RigVMFunction_MathDoubleClamp(Value=@Add.Result).Result
    A=&RigVMFunction_MathDoubleClamp(Value=&RigVMFunction_MathDoubleAdd(B=@n22.Result, A=@n21.Result).Result).Result
   )
   n23 = &RigVMFunction_MathDoubleClamp(Value=@Subtract.Result)
   /Script/ControlRig.RigUnit_AimItem(
    DebugSettings=(bEnabled=False,Scale=10,WorldOffset=(Rotation=(X=0,Y=0,Z=0,W=1),Translation=(X=0,Y=0,Z=0),Scale3D=(X=1,Y=1,Z=1)))
    Item=(Type=Bone,Name="dyn_C_front_robe_1")
    Primary=(Weight=1,Axis=(X=1,Y=0,Z=0),Target=(X=37.6193,Y=-13.7034,Z=-20.5058),Kind=Location,Space=(Type=Bone,Name="thigh_r"))
    Secondary=(Weight=0,Axis=(X=0,Y=0,Z=0),Target=(X=0,Y=0,Z=0),Kind=Location,Space=(Type=Bone,Name="None"))
    ExecutePin=@RigVMFunction_Sequence.B
    Weight=&RigVMFunction_MathDoubleSub(B=&RigVMFunction_MathDoubleMul(B=0.5, A=@n18.Value).Result, A=@n23.Result).Result
   ) {is=struct RigVMUnitNode}
   > &RigUnit_AimItem(
    DebugSettings=(bEnabled=False,Scale=10,WorldOffset=(Rotation=(X=0,Y=0,Z=0,W=1),Translation=(X=0,Y=0,Z=0),Scale3D=(X=1,Y=1,Z=1)))
    Item=(Type=Bone,Name="dyn_C_front_robe_1")
    Primary=(Weight=1,Axis=(X=1,Y=0,Z=0),Target=(X=-38.101,Y=12.3571,Z=13.4983),Kind=Location,Space=(Type=Bone,Name="thigh_l"))
    Secondary=(Weight=0,Axis=(X=0,Y=0,Z=0),Target=(X=0,Y=0,Z=0),Kind=Location,Space=(Type=Bone,Name="None"))
    Weight=&RigVMFunction_MathDoubleAdd(A=&RigVMFunction_MathDoubleSub(A=1, B=@n23.Result).Result, B=RigVMVariableNode() {variable=isPlayingFrontendIdle}.Value).Result
   )
   > &RigUnit_ModifyTransforms(
    ItemToModify=((Item=(Type=Bone,Name="dyn_C_front_robe_1"),Transform=(Rotation=(X=0,Y=0.0175,Z=0,W=0.9998),Translation=(X=0,Y=0,Z=1),Scale3D=(X=1,Y=1,Z=1))))
    Weight=1
    WeightMinimum=0
   )
   > &RigUnit_ParentConstraint(
    AdvancedSettings=(InterpolationType=Average,RotationOrderForFilter=XZY)
    Child=(Type=Bone,Name="dyn_L_front_A_robe_1")
    Filter=(TranslationFilter=(bX=true,bY=true,bZ=true),RotationFilter=(bX=true,bY=true,bZ=true),ScaleFilter=(bX=True,bY=True,bZ=True))
    Parents=((Item=(Type=Bone,Name="thigh_l"),Weight=1))
    Weight=1
   )
   > &RigUnit_ModifyTransforms(
    ItemToModify=((Item=(Type=Bone,Name="dyn_L_front_A_robe_1"),Transform=(Rotation=(X=-0.0511,Y=-0.1275,Z=-0.0662,W=0.9883),Translation=(X=0,Y=0,Z=0),Scale3D=(X=1,Y=1,Z=1))))
    WeightMinimum=0
    Weight=&RigVMFunction_MathDoubleAdd(B=RigVMVariableNode() {variable=L thighSide alpha}.Value, A=RigVMVariableNode() {variable=R thighForward alpha}.Value).Result
   )
   > &RigUnit_ParentConstraint(
    AdvancedSettings=(InterpolationType=Average,RotationOrderForFilter=XZY)
    Child=(Type=Bone,Name="dyn_L_side_robe_1")
    Filter=(TranslationFilter=(bX=true,bY=true,bZ=true),RotationFilter=(bX=true,bY=true,bZ=true),ScaleFilter=(bX=True,bY=True,bZ=True))
    Parents=((Item=(Type=Bone,Name="thigh_l"),Weight=1))
    Weight=0.75
   )
   > &RigUnit_ModifyTransforms(
    ItemToModify=((Item=(Type=Bone,Name="dyn_L_side_robe_1"),Transform=(Rotation=(X=0,Y=-0.0436,Z=0,W=0.999),Translation=(X=0,Y=0,Z=0),Scale3D=(X=1,Y=1,Z=1))))
    WeightMinimum=0
    Weight=&RigVMFunction_MathDoubleAdd(A=RigVMVariableNode() {variable=L thighSide alpha}.Value, B=RigVMVariableNode() {variable=L thighForward alpha}.Value).Result
   )
   > &RigUnit_ParentConstraint(
    AdvancedSettings=(InterpolationType=Average,RotationOrderForFilter=XZY)
    Child=(Type=Bone,Name="dyn_L_back_robe_2")
    Filter=(TranslationFilter=(bX=true,bY=true,bZ=true),RotationFilter=(bX=true,bY=true,bZ=true),ScaleFilter=(bX=True,bY=True,bZ=True))
    Parents=((Item=(Type=Bone,Name="thigh_l"),Weight=1))
    Weight=&RigVMFunction_MathDoubleMul(B=0.4, A=RigVMVariableNode() {variable=L thighForward alpha}.Value).Result
   )
   > &RigUnit_ModifyTransforms(
    ItemToModify=((Item=(Type=Bone,Name="dyn_L_back_robe_1"),Transform=(Rotation=(X=0,Y=-0.2164,Z=0,W=0.9763),Translation=(X=0,Y=0,Z=1),Scale3D=(X=1,Y=1,Z=1))))
    WeightMinimum=0
    Weight=RigVMVariableNode() {variable=L thighBack alpha}.Value
   )
   > &RigUnit_ParentConstraint(
    AdvancedSettings=(InterpolationType=Average,RotationOrderForFilter=XZY)
    Child=(Type=Bone,Name="dyn_R_front_B_robe_1")
    Filter=(TranslationFilter=(bX=true,bY=true,bZ=true),RotationFilter=(bX=true,bY=true,bZ=true),ScaleFilter=(bX=True,bY=True,bZ=True))
    Parents=((Item=(Type=Bone,Name="thigh_r"),Weight=1))
    Weight=1
   )
   > &RigUnit_ModifyTransforms(
    ItemToModify=((Item=(Type=Bone,Name="dyn_R_front_B_robe_1"),Transform=(Rotation=(X=-0.0396,Y=-0.0889,Z=0.0396,W=0.9945),Translation=(X=0,Y=0,Z=0),Scale3D=(X=1,Y=1,Z=1))))
    WeightMinimum=0
    Weight=&RigVMFunction_MathDoubleAdd(B=RigVMVariableNode() {variable=L thighSide alpha}.Value, A=RigVMVariableNode() {variable=R thighForward alpha}.Value).Result
   )
   > &RigUnit_ParentConstraint(
    AdvancedSettings=(InterpolationType=Average,RotationOrderForFilter=XZY)
    Child=(Type=Bone,Name="dyn_R_side_robe_1")
    Filter=(TranslationFilter=(bX=true,bY=true,bZ=true),RotationFilter=(bX=true,bY=true,bZ=true),ScaleFilter=(bX=True,bY=True,bZ=True))
    Parents=((Item=(Type=Bone,Name="thigh_r"),Weight=1))
    Weight=1
   )
   > &RigUnit_ParentConstraint(
    AdvancedSettings=(InterpolationType=Average,RotationOrderForFilter=XZY)
    Child=(Type=Bone,Name="dyn_R_side_robe_2")
    Filter=(TranslationFilter=(bX=true,bY=true,bZ=true),RotationFilter=(bX=true,bY=true,bZ=true),ScaleFilter=(bX=True,bY=True,bZ=True))
    Parents=((Item=(Type=Bone,Name="thigh_r"),Weight=1))
    Weight=1
   )
   > &RigUnit_ModifyTransforms(
    ItemToModify=((Item=(Type=Bone,Name="dyn_R_side_robe_1"),Transform=(Rotation=(X=0,Y=-0.0436,Z=0,W=0.999),Translation=(X=0,Y=0,Z=0),Scale3D=(X=1,Y=1,Z=1))))
    WeightMinimum=0
    Weight=&RigVMFunction_MathDoubleAdd(A=RigVMVariableNode() {variable=R thighSide alpha}.Value, B=RigVMVariableNode() {variable=R thighForward alpha}.Value).Result
   )
   > &RigUnit_ParentConstraint(
    AdvancedSettings=(InterpolationType=Average,RotationOrderForFilter=XZY)
    Child=(Type=Bone,Name="dyn_R_back_robe_2")
    Filter=(TranslationFilter=(bX=false,bY=false,bZ=false),RotationFilter=(bX=true,bY=true,bZ=true),ScaleFilter=(bX=True,bY=True,bZ=True))
    Parents=((Item=(Type=Bone,Name="thigh_r"),Weight=1))
    Weight=&RigVMFunction_MathDoubleMul(B=0.4, A=RigVMVariableNode() {variable=R thighForward alpha}.Value).Result
   )
   > &RigUnit_ModifyTransforms(
    ItemToModify=((Item=(Type=Bone,Name="dyn_R_back_robe_1"),Transform=(Rotation=(X=0,Y=-0.2164,Z=0,W=0.9763),Translation=(X=0,Y=0,Z=1),Scale3D=(X=1,Y=1,Z=1))))
    WeightMinimum=0
    Weight=RigVMVariableNode() {variable=R thighBack alpha}.Value
   )
   n24 = &RigUnit_GetRelativeTransformForItem(Child=(Type=Bone,Name="dyn_digitigrade_calf_l"), Parent=(Type=Bone,Name="thigh_l"), bChildInitial=False, bParentInitial=False)
   > RigVMVariableNode(Value=&RigVMFunction_MathDoubleRemap(
     SourceMaximum=130
     SourceMinimum=120
     TargetMaximum=1
     TargetMinimum=0
     Value=&RigVMFunction_MathQuaternionToEuler(RotationOrder=ZYX, Value=@n24.RelativeTransform.Rotation).Result.Z
    ).Result) {variable=L calf bend}
   n25 = &RigUnit_GetRelativeTransformForItem(Child=(Type=Bone,Name="dyn_digitigrade_calf_r"), Parent=(Type=Bone,Name="thigh_r"), bChildInitial=False, bParentInitial=False)
   > RigVMVariableNode(Value=&RigVMFunction_MathDoubleRemap(
     SourceMaximum=130
     SourceMinimum=120
     TargetMaximum=1
     TargetMinimum=0
     Value=&RigVMFunction_MathQuaternionToEuler(RotationOrder=ZYX, Value=@n25.RelativeTransform.Rotation).Result.Z
    ).Result) {variable=R calf bend}
   n26 = &RigUnit_GetRelativeTransformForItem(Child=(Type=Bone,Name="thigh_l"), Parent=(Type=Bone,Name="pelvis"), bChildInitial=False, bParentInitial=False)
   > RigVMVariableNode(Value=&RigVMFunction_MathDoubleRemap(
     SourceMaximum=-25
     SourceMinimum=-45
     TargetMaximum=1
     TargetMinimum=0
     Value=&RigVMFunction_MathQuaternionToEuler(RotationOrder=ZYX, Value=@n26.RelativeTransform.Rotation).Result.Z
    ).Result) {variable=L thigh max rotation}
   n27 = &RigUnit_GetRelativeTransformForItem(Child=(Type=Bone,Name="thigh_r"), Parent=(Type=Bone,Name="pelvis"), bChildInitial=False, bParentInitial=False)
   > RigVMVariableNode(Value=&RigVMFunction_MathDoubleRemap(
     SourceMaximum=170
     SourceMinimum=150
     TargetMaximum=1
     TargetMinimum=0
     Value=&RigVMFunction_MathQuaternionToEuler(RotationOrder=ZYX, Value=@n27.RelativeTransform.Rotation).Result.Z
    ).Result) {variable=R thigh max rotation}
   > &RigUnit_ModifyTransforms(
    ItemToModify=((Item=(Type=Bone,Name="dyn_L_back_robe_1"),Transform=(Rotation=(X=0,Y=-0.766,Z=0,W=0.6428),Translation=(X=0,Y=0,Z=0),Scale3D=(X=1,Y=1,Z=1))))
    WeightMinimum=0
    Weight=&RigVMFunction_MathDoubleMul(B=RigVMVariableNode() {variable=L calf bend}.Value, A=RigVMVariableNode() {variable=L thigh max rotation}.Value).Result
   )
   > &RigUnit_ModifyTransforms(
    ItemToModify=((Item=(Type=Bone,Name="dyn_R_back_robe_1"),Transform=(Rotation=(X=0,Y=-0.766,Z=0,W=0.6428),Translation=(X=0,Y=0,Z=0),Scale3D=(X=1,Y=1,Z=1))))
    WeightMinimum=0
    Weight=&RigVMFunction_MathDoubleMul(A=RigVMVariableNode() {variable=R thigh max rotation}.Value, B=RigVMVariableNode() {variable=R calf bend}.Value).Result
   )
   n28 = RigVMVariableNode() {variable=L calf bend}
   n29 = RigVMVariableNode() {variable=L thigh max rotation}
   n30 = RigVMVariableNode() {variable=R thigh max rotation}
   n31 = RigVMVariableNode() {variable=R calf bend}
   Divide = &RigVMFunction_MathDoubleDiv(
    B=2
    A=&RigVMFunction_MathDoubleAdd(
     B=&RigVMFunction_MathDoubleMul(B=@n28.Value, A=@n29.Value).Result
     A=&RigVMFunction_MathDoubleMul(A=@n30.Value, B=@n31.Value).Result
    ).Result
   )
   > &RigUnit_RotationConstraintLocalSpaceOffset(
    AdvancedSettings=(InterpolationType=Average,RotationOrderForFilter=XZY)
    Child=(Type=Bone,Name="dyn_C_back_robe_1")
    Filter=(bX=True,bY=True,bZ=True)
    Parents=((Item=(Type=Bone,Name="dyn_L_back_robe_1"),Weight=0.5),(Item=(Type=Bone,Name="dyn_R_back_robe_1"),Weight=0.5))
    Weight=@Divide.Result
   )
   AlphaInterp = &RigVMFunction_AlphaInterp(
    InterpSpeedDecreasing=4
    InterpSpeedIncreasing=10
    Scale=1
    bClampResult=False
    Value=RigVMVariableNode() {variable=Is Playing Melee Anim Alpha}.Value
   )
   > &RigUnit_ModifyTransforms(
    ItemToModify=((Item=(Type=Bone,Name="dyn_C_front_robe_1"),Transform=(Rotation=(X=-0.1736,Y=-0.0215,Z=-0.0038,W=0.9846),Translation=(X=0,Y=0,Z=2),Scale3D=(X=1,Y=1,Z=1))))
    WeightMinimum=0
    Weight=@AlphaInterp.Result
   )
   > &RigUnit_ModifyTransforms(
    ItemToModify=((Item=(Type=Bone,Name="dyn_L_front_A_robe_1"),Transform=(Rotation=(X=0,Y=-0.0218,Z=0,W=0.9998),Translation=(X=0,Y=0,Z=2),Scale3D=(X=1,Y=1,Z=1))))
    WeightMinimum=0
    Weight=@AlphaInterp.Result
   )
   > &RigUnit_ModifyTransforms(
    ItemToModify=((Item=(Type=Bone,Name="dyn_L_side_robe_1"),Transform=(Rotation=(X=-0.1736,Y=-0.0215,Z=-0.0038,W=0.9846),Translation=(X=0,Y=0,Z=2),Scale3D=(X=1,Y=1,Z=1))))
    WeightMinimum=0
    Weight=@AlphaInterp.Result
   )
   &RigVMFunction_MathDoubleAbs(Value=0)
   /Script/RigVM.RigVMFunction_MathDoubleLess(A=0, B=0) {is=struct RigVMUnitNode, template="Less::Execute(in A,in B,out Result)"}
  ]}
]
hierarchy [
 root {loc=<0 0 0>, rot=<0 0 0 1>}
 pelvis {parent=root, loc=<0 -0.7071 93.1191>, rot=<0.0127 0.707 -0.0127 -0.707>}
 spine_01 {parent=pelvis, loc=<6.7494 -0.6388 0>, rot=<0 0 0.034 -0.9994>}
 spine_02 {parent=spine_01, loc=<7.2518 -0.035 0>, rot=<0 0 0.0238 -0.9997>}
 spine_03 {parent=spine_02, loc=<6.0104 0 0>, rot=<0 0 0.0149 -0.9999>}
 spine_04 {parent=spine_03, loc=<8.1024 0 0>, rot=<0 0 -0.0132 -0.9999>}
 spine_05 {parent=spine_04, loc=<10.0146 0 0>, rot=<0 0 -0.0015 -1>}
 attach_cape {parent=spine_05, loc=<9.017 -11.714 0>, rot=<0.0616 -0.7044 -0.0616 -0.7044>}
 neck_01 {parent=spine_05, loc=<11.3138 0.1529 0>, rot=<0 0 -0.2003 -0.9797>}
 neck_02 {parent=neck_01, loc=<4.5627 0 0>, rot=<0 0 0.0504 -0.9987>}
 head {parent=neck_02, loc=<5.3914 0 0>, rot=<0 0 0.115 -0.9934>}
 attach_backpack {parent=spine_05, loc=<-0.123 -13.473 0>, rot=<-0.0093 -0.707 0.0093 -0.707>}
 clavicle_l {parent=spine_05, loc=<4.3575 -0.1466 -2.3223>, rot=<0.0465 -0.7289 -0.009 -0.6829>}
 upperarm_l {parent=clavicle_l, loc=<15.5111 0 0>, rot=<0.0226 -0.3234 -0.0088 -0.9459>}
 lowerarm_l {parent=upperarm_l, loc=<26.6514 0 0>, rot=<0.0391 -0.0099 -0.2453 -0.9686>}
 hand_l {parent=lowerarm_l, loc=<22.3491 0 0>, rot=<-0.7081 -0.0507 0.0975 -0.6975>}
 thumb_01_l {parent=hand_l, loc=<1.3118 -0.3191 -2.459>, rot=<0.6546 -0.4042 0.0664 -0.6354>}
 thumb_02_l {parent=thumb_01_l, loc=<5.9336 -0.0448 0.0401>, rot=<-0.0037 0.0213 0.298 -0.9543>}
 thumb_03_l {parent=thumb_02_l, loc=<2.2223 0.3682 -0.0907>, rot=<0.0118 -0.013 0.0066 -0.9998>}
 index_metacarpal_l {parent=hand_l, loc=<1.6535 0.0183 -2.96>, rot=<0.0158 -0.0124 -0.0173 -0.9996>}
 index_01_l {parent=index_metacarpal_l, loc=<8.3008 0.0357 -0.0524>, rot=<-0.0044 0.0485 0.0924 -0.9945>}
 index_02_l {parent=index_01_l, loc=<3.3587 -0.3004 0.0452>, rot=<-0.002 0.0027 0.3225 -0.9466>}
 index_03_l {parent=index_02_l, loc=<2.5578 -0.123 -0.0252>, rot=<-0.002 0.0067 0.2558 -0.9667>}
 middle_metacarpal_l {parent=hand_l, loc=<2.1856 -0.1725 -0.5984>, rot=<0.006 0.0168 -0.0329 -0.9993>}
 middle_01_l {parent=middle_metacarpal_l, loc=<7.3275 -0.2447 0.0129>, rot=<-0.0208 0.0042 0.0858 -0.9961>}
 middle_02_l {parent=middle_01_l, loc=<4.0123 0.1401 0.1175>, rot=<-0.0218 0.0258 0.3512 -0.9357>}
 middle_03_l {parent=middle_02_l, loc=<2.6077 0.0534 0.1373>, rot=<0.0082 0.0188 0.1708 -0.9851>}
 ring_metacarpal_l {parent=hand_l, loc=<2.2074 -0.3625 1.3638>, rot=<-0.0284 0.0722 -0.0458 -0.9959>}
 ring_01_l {parent=ring_metacarpal_l, loc=<7.0295 -0.6197 -0.1446>, rot=<-0.0489 -0.0252 0.1892 -0.9804>}
 ring_02_l {parent=ring_01_l, loc=<3.6174 0.4647 -0.1027>, rot=<0 0 0.2806 -0.9598>}
 ring_03_l {parent=ring_02_l, loc=<2.4152 0.114 0.0199>, rot=<0 0 0.1189 -0.9929>}
 pinky_metacarpal_l {parent=hand_l, loc=<1.7616 -0.9753 3.3814>, rot=<-0.0397 0.0814 -0.069 -0.9935>}
 pinky_01_l {parent=pinky_metacarpal_l, loc=<7.1709 -0.4831 -0.1094>, rot=<-0.0654 -0.0035 0.2242 -0.9723>}
 pinky_02_l {parent=pinky_01_l, loc=<2.8274 0.0893 -0.1043>, rot=<0.0037 0.0011 0.2854 -0.9584>}
 pinky_03_l {parent=pinky_02_l, loc=<2.2046 0.0307 -0.0078>, rot=<-0.0002 0.0098 0.1445 -0.9895>}
 lowerarm_twist_02_l {parent=lowerarm_l, loc=<9.6199 0 0>, rot=<0 0 0 -1>}
 lowerarm_twist_01_l {parent=lowerarm_l, loc=<17.8549 0 0>, rot=<0 0 0 -1>}
 dyn_L_topSleeve_A {parent=lowerarm_l, loc=<-0.4674 1.9396 4.8305>, rot=<-0.0279 -0.9912 0.1291 0.0021>}
 dyn_L_topSleeve_B {parent=dyn_L_topSleeve_A, loc=<-11.3177 0 0>, rot=<0 0 0 1>}
 dyn_L_bottomSleeve_A {parent=lowerarm_l, loc=<0.9239 0.0202 -6.9732>, rot=<0.0725 0.1268 0.9886 0.037>}
 dyn_L_bottomSleeve_B {parent=dyn_L_bottomSleeve_A, loc=<-11.3182 0 -0.0004>, rot=<0 0 0 1>}
 dyn_L_frontSleeve_A {parent=lowerarm_l, loc=<1.3454 6.1344 -4.0004>, rot=<0.0377 -0.9932 0.1105 0.0045>}
 dyn_L_frontSleeve_B {parent=dyn_L_frontSleeve_A, loc=<-11.3176 0 -0.0001>, rot=<0 0 0 1>}
 dyn_L_backSleeve_A {parent=lowerarm_l, loc=<0.378 -5.5866 -0.4899>, rot=<0.0676 0.0228 0.9879 0.1377>}
 dyn_L_backSleeve_B {parent=dyn_L_backSleeve_A, loc=<-11.3178 0 0>, rot=<0 0 0 1>}
 upperarm_twist_01_l {parent=upperarm_l, loc=<0.0041 0 0>, rot=<0 0 0 1>}
 upperarm_twist_02_l {parent=upperarm_l, loc=<13.7941 0 0>, rot=<0 0 0 1>}
 clavicle_r {parent=spine_05, loc=<4.3575 -0.1466 2.3223>, rot=<0.7289 0.0465 -0.6829 0.009>}
 upperarm_r {parent=clavicle_r, loc=<-15.5111 0 0>, rot=<0.0226 -0.3234 -0.0088 -0.9459>}
 lowerarm_r {parent=upperarm_r, loc=<-26.6514 0 0>, rot=<0.0391 -0.0099 -0.2453 -0.9686>}
 hand_r {parent=lowerarm_r, loc=<-22.3491 0 0>, rot=<-0.7081 -0.0507 0.0975 -0.6975>}
 thumb_01_r {parent=hand_r, loc=<-1.3118 0.3191 2.459>, rot=<0.6546 -0.4042 0.0664 -0.6354>}
 thumb_02_r {parent=thumb_01_r, loc=<-5.9336 0.0448 -0.0401>, rot=<-0.0037 0.0213 0.298 -0.9543>}
 thumb_03_r {parent=thumb_02_r, loc=<-2.2223 -0.3682 0.0907>, rot=<0.0118 -0.013 0.0066 -0.9998>}
 index_metacarpal_r {parent=hand_r, loc=<-1.6535 -0.0183 2.96>, rot=<0.0158 -0.0124 -0.0173 -0.9996>}
 index_01_r {parent=index_metacarpal_r, loc=<-8.3008 -0.0357 0.0524>, rot=<-0.0044 0.0485 0.0924 -0.9945>}
 index_02_r {parent=index_01_r, loc=<-3.3587 0.3004 -0.0452>, rot=<-0.002 0.0027 0.3225 -0.9466>}
 index_03_r {parent=index_02_r, loc=<-2.5578 0.123 0.0252>, rot=<-0.002 0.0067 0.2558 -0.9667>}
 middle_metacarpal_r {parent=hand_r, loc=<-2.1856 0.1725 0.5984>, rot=<0.006 0.0168 -0.0329 -0.9993>}
 middle_01_r {parent=middle_metacarpal_r, loc=<-7.3275 0.2447 -0.0129>, rot=<-0.0208 0.0042 0.0858 -0.9961>}
 middle_02_r {parent=middle_01_r, loc=<-4.0123 -0.1401 -0.1175>, rot=<-0.0218 0.0258 0.3512 -0.9357>}
 middle_03_r {parent=middle_02_r, loc=<-2.6077 -0.0534 -0.1373>, rot=<0.0082 0.0188 0.1708 -0.9851>}
 ring_metacarpal_r {parent=hand_r, loc=<-2.2074 0.3625 -1.3638>, rot=<-0.0284 0.0722 -0.0458 -0.9959>}
 ring_01_r {parent=ring_metacarpal_r, loc=<-7.0295 0.6197 0.1446>, rot=<-0.0489 -0.0252 0.1892 -0.9804>}
 ring_02_r {parent=ring_01_r, loc=<-3.6174 -0.4647 0.1027>, rot=<0 0 0.2806 -0.9598>}
 ring_03_r {parent=ring_02_r, loc=<-2.4152 -0.114 -0.0199>, rot=<0 0 0.1189 -0.9929>}
 pinky_metacarpal_r {parent=hand_r, loc=<-1.7616 0.9753 -3.3814>, rot=<-0.0397 0.0814 -0.069 -0.9935>}
 pinky_01_r {parent=pinky_metacarpal_r, loc=<-7.1709 0.4831 0.1094>, rot=<-0.0654 -0.0035 0.2242 -0.9723>}
 pinky_02_r {parent=pinky_01_r, loc=<-2.8274 -0.0893 0.1043>, rot=<0.0037 0.0011 0.2854 -0.9584>}
 pinky_03_r {parent=pinky_02_r, loc=<-2.2046 -0.0307 0.0078>, rot=<-0.0002 0.0098 0.1445 -0.9895>}
 lowerarm_twist_02_r {parent=lowerarm_r, loc=<-9.6199 0 0>, rot=<0 0 0 -1>}
 lowerarm_twist_01_r {parent=lowerarm_r, loc=<-17.8549 0 0>, rot=<0 0 0 -1>}
 dyn_R_frontSleeve_A {parent=lowerarm_r, loc=<-1.3454 -6.1344 4.0004>, rot=<0.0377 -0.9932 0.1105 0.0045>}
 dyn_R_frontSleeve_B {parent=dyn_R_frontSleeve_A, loc=<11.3177 0 -0.0001>, rot=<0 0 0 1>}
 dyn_R_backSleeve_A {parent=lowerarm_r, loc=<-0.3781 5.5866 0.4899>, rot=<0.0676 0.0228 0.9879 0.1377>}
 dyn_R_backSleeve_B {parent=dyn_R_backSleeve_A, loc=<11.3177 0 0>, rot=<0 0 0 1>}
 dyn_R_topSleeve_A {parent=lowerarm_r, loc=<0.4674 -1.9396 -4.8305>, rot=<-0.0279 -0.9912 0.1291 0.0021>}
 dyn_R_topSleeve_B {parent=dyn_R_topSleeve_A, loc=<11.3179 0 -0.0002>, rot=<0 0 0 1>}
 dyn_R_bottomSleeve_A {parent=lowerarm_r, loc=<-0.9236 -0.0203 6.9729>, rot=<0.0725 0.1268 0.9886 0.037>}
 dyn_R_bottomSleeve_B {parent=dyn_R_bottomSleeve_A, loc=<11.3184 0 0.0005>, rot=<0 0 0 1>}
 upperarm_twist_01_r {parent=upperarm_r, loc=<-0.0041 0 0>, rot=<0 0 0 -1>}
 upperarm_twist_02_r {parent=upperarm_r, loc=<-13.7941 0 0>, rot=<0 0 0 -1>}
 thigh_l {parent=pelvis, loc=<-2.5693 -0.0109 -7.7501>, rot=<0.0012 0.0572 0.0155 -0.9982>}
 calf_l {parent=thigh_l, loc=<-39.0484 0 0>, rot=<0.0006 -0.0161 -0.0391 -0.9991>}
 foot_l {parent=calf_l, loc=<-42.561 0.1077 0.4292>, rot=<0.0418 -0.0455 0.0343 -0.9975>}
 ball_l {parent=foot_l, loc=<-7.4164 8.9523 0.2306>, rot=<0.0001 -0.0001 -0.7063 -0.7079>}
 calf_twist_02_l {parent=calf_l, loc=<-18.8064 -0.0967 0.0397>, rot=<0.0237 0.0095 -0.0035 -0.9997>}
 calf_twist_01_l {parent=calf_l, loc=<-35.813 -0.0967 0.0397>, rot=<0.0237 0.0095 -0.0035 -0.9997>}
 thigh_twist_01_l {parent=thigh_l, loc=<-21.6303 0 0>, rot=<0.0013 0 0 -1>}
 dyn_digitigrade_calf_l {parent=thigh_l, loc=<-39.0484 0 0>, rot=<0.0006 -0.0161 -0.0391 -0.9991>}
 dyn_digitigrade_foot_l {parent=dyn_digitigrade_calf_l, loc=<-30.0317 -5.9157 -0.9065>, rot=<-0.0031 -0.0411 0.0414 -0.9983>}
 dyn_digitigrade_heel_l {parent=dyn_digitigrade_foot_l, loc=<-15.7286 5.1003 -0.703>, rot=<0.0453 -0.0026 -0.0052 -0.999>}
 dyn_digitigrade_toe_l {parent=dyn_digitigrade_heel_l, loc=<-2.853 9.1506 0.3209>, rot=<-0.0453 0.0026 0.0052 -0.999>}
 thigh_r {parent=pelvis, loc=<-2.5693 -0.0109 7.7501>, rot=<0.0572 -0.0012 0.9982 0.0155>}
 calf_r {parent=thigh_r, loc=<39.0484 0 0>, rot=<0.0006 -0.0161 -0.0391 -0.9991>}
 foot_r {parent=calf_r, loc=<42.561 -0.1077 -0.4292>, rot=<0.0418 -0.0455 0.0343 -0.9975>}
 ball_r {parent=foot_r, loc=<7.4164 -8.9523 -0.2306>, rot=<0.0001 -0.0001 -0.7063 -0.7079>}
 calf_twist_02_r {parent=calf_r, loc=<18.8064 0.0967 -0.0397>, rot=<0.0237 0.0095 -0.0035 -0.9997>}
 calf_twist_01_r {parent=calf_r, loc=<35.813 0.0967 -0.0397>, rot=<0.0237 0.0095 -0.0035 -0.9997>}
 thigh_twist_01_r {parent=thigh_r, loc=<21.6303 0 0>, rot=<0.0013 0 0 -1>}
 dyn_digitigrade_calf_r {parent=thigh_r, loc=<39.0484 0 0>, rot=<0.0006 -0.0161 -0.0391 -0.9991>}
 dyn_digitigrade_foot_r {parent=dyn_digitigrade_calf_r, loc=<30.0317 5.9157 0.9065>, rot=<-0.0031 -0.0411 0.0414 -0.9983>}
 dyn_digitigrade_heel_r {parent=dyn_digitigrade_foot_r, loc=<15.7286 -5.1003 0.7031>, rot=<0.0453 -0.0026 -0.0052 -0.999>}
 dyn_digitigrade_toe_r {parent=dyn_digitigrade_heel_r, loc=<2.853 -9.1506 -0.321>, rot=<-0.0453 0.0026 0.0052 -0.999>}
 dyn_C_front_robe_1 {parent=pelvis, loc=<4.3547 9.6758 -0.6374>, rot=<0.0421 0.7059 0.7059 0.0421>}
 dyn_C_front_robe_2 {parent=dyn_C_front_robe_1, loc=<17.1984 0 0>, rot=<0 0 0 1>}
 dyn_C_front_robe_3 {parent=dyn_C_front_robe_2, loc=<15.8747 0 0>, rot=<0 0 0 1>}
 dyn_C_back_robe_1 {parent=pelvis, loc=<3.5927 -12.0588 0>, rot=<0.0551 -0.705 0.705 -0.0551>}
 dyn_C_back_robe_2 {parent=dyn_C_back_robe_1, loc=<11.4052 0 0>, rot=<0 0 0 1>}
 dyn_C_back_robe_3 {parent=dyn_C_back_robe_2, loc=<13.1321 0 0>, rot=<0 0 0 1>}
 dyn_R_front_B_robe_1 {parent=pelvis, loc=<4.0429 4.1318 13.0343>, rot=<0.1009 0.3028 0.9474 0.0232>}
 dyn_R_front_B_robe_2 {parent=dyn_R_front_B_robe_1, loc=<15.1663 0 0>, rot=<0 0 0 1>}
 dyn_R_front_B_robe_3 {parent=dyn_R_front_B_robe_2, loc=<16.6118 0 0>, rot=<0 0 0 1>}
 dyn_R_side_robe_1 {parent=pelvis, loc=<3.8436 -4.238 14.6644>, rot=<0.103 -0.0895 0.9906 0.0085>}
 dyn_R_side_robe_2 {parent=dyn_R_side_robe_1, loc=<13.5623 0 0>, rot=<0 0 0 1>}
 dyn_R_side_robe_3 {parent=dyn_R_side_robe_2, loc=<13.2054 0 0>, rot=<0 0 0 1>}
 dyn_R_back_robe_1 {parent=pelvis, loc=<3.7103 -9.9755 9.9572>, rot=<0.0704 -0.4139 0.9067 -0.0395>}
 dyn_R_back_robe_2 {parent=dyn_R_back_robe_1, loc=<11.4052 0 0>, rot=<0 0 0 1>}
 dyn_R_back_robe_3 {parent=dyn_R_back_robe_2, loc=<13.1321 0 0>, rot=<0 0 0 1>}
 dyn_L_side_robe_1 {parent=pelvis, loc=<3.9811 -2.0463 -15.9024>, rot=<0.0157 -0.9905 0.092 -0.1008>}
 dyn_L_side_robe_2 {parent=dyn_L_side_robe_1, loc=<13.5623 0 0.0001>, rot=<0 0 0 1>}
 dyn_L_side_robe_3 {parent=dyn_L_side_robe_2, loc=<13.2054 0 0>, rot=<0 0 0 1>}
 dyn_L_back_robe_1 {parent=pelvis, loc=<3.7103 -9.9755 -9.9572>, rot=<0.0395 -0.9067 0.4139 -0.0704>}
 dyn_L_back_robe_2 {parent=dyn_L_back_robe_1, loc=<11.4053 0 0>, rot=<0 0 0 1>}
 dyn_L_back_robe_3 {parent=dyn_L_back_robe_2, loc=<13.1321 0 0.0001>, rot=<0 0 0 1>}
 dyn_L_front_A_robe_1 {parent=pelvis, loc=<4.4588 7.2074 -9.9332>, rot=<-0.0034 -0.8229 -0.5635 -0.0731>}
 dyn_L_front_A_robe_2 {parent=dyn_L_front_A_robe_1, loc=<17.3902 0 0>, rot=<0 0 0 1>}
 dyn_L_front_A_robe_3 {parent=dyn_L_front_A_robe_2, loc=<16.3586 0 0>, rot=<0 0 0 1>}
 dyn_belt_base {parent=pelvis, loc=<11.8149 11.1712 0.3521>, rot=<0.0414 0.7059 0.7069 -0.0161>}
 dyn_L_belt_1 {parent=dyn_belt_base, loc=<8.1676 -3.5504 0.1523>, rot=<-0.0034 0.0288 0.1962 -0.9801>}
 dyn_L_belt_2 {parent=dyn_L_belt_1, loc=<6.6925 0 0>, rot=<0.009 -0.0276 0.0002 -0.9996>}
 dyn_R_belt_1 {parent=dyn_belt_base, loc=<5.1665 -0.2489 0.9936>, rot=<0.0022 0.0234 0.0133 -0.9996>}
 dyn_R_belt_2 {parent=dyn_R_belt_1, loc=<9.5188 -0.0248 -0.0342>, rot=<0.0001 -0.0189 0.0137 -0.9997>}
 dyn_R_belt_3 {parent=dyn_R_belt_2, loc=<8.7615 0 -0.0256>, rot=<0.0003 -0.0163 0 -0.9999>}
]
variables [
 Is Playing Melee Anim Alpha {type=double, default=0}
 L calf bend {type=double, default=0}
 L calfBend alpha {type=double, default=0}
 L thigh max rotation {type=double, default=0}
 L thighBack alpha {type=float, default=0}
 L thighForward alpha {type=float, default=0}
 L thighSide alpha {type=float, default=0}
 L thighVeryForward {type=double, default=0}
 L_thigh_Z_value {type=double, default=0}
 L_thigh_rotated_more_than_R {type=double, default=0}
 R calf bend {type=double, default=0}
 R calfBend alpha {type=double, default=0}
 R thigh max rotation {type=double, default=0}
 R thighBack alpha {type=float, default=0}
 R thighForward alpha {type=float, default=0}
 R thighSide alpha {type=float, default=0}
 R thighVeryForward {type=double, default=0}
 R_thigh_Z_value {type=double, default=0}
 R_thigh_rotated_more_than_L {type=double, default=0}
 isPlayingFrontendIdle {type=double, default=0}
]
